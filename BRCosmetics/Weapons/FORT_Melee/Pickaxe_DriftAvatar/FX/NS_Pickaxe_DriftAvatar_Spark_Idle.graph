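# Idle Niagara effect for the DriftAvatar pickaxe, attaching a pink glow sprite and crackling electric mesh sparks to the supplied skeletal mesh.
# Glow pins a camera-offset sprite, with alpha scaled by time of day.
# Shoulders_FX002 repeatedly bursts random bolts, some traveling in gameplay but none in front end, animated through material mask parameters, dimmed by system flicker, with separate glow attributes.
# Light and older Shoulders_FX emitters are disabled.
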NiagaraSystem /BRCosmetics/Weapons/FORT_Melee/Pickaxe_DriftAvatar/FX/NS_Pickaxe_DriftAvatar_Spark_Idle
dependencies {
 usedDataInterfaces [&NiagaraDataInterfaceArrayFloat3, &NiagaraDataInterfaceCurve, &NiagaraDataInterfaceArrayFloat]
 usedDynamicInputs [
  &AddVectorToPosition
  &Lerp_Vector
  &MakeFloatFromInt
  &Boolean_NotOperation
 ]
 usedModules [
  &self:SystemSpawnScript.SystemScriptSource.SystemScriptGraph.NiagaraNodeAssignment_1.SetVariables_#1
  &SystemState
  &FN_Timer
  &self:SystemSpawnScript.SystemScriptSource.SystemScriptGraph.NiagaraNodeAssignment_0.SetVariables_#2
  &self:SystemSpawnScript.SystemScriptSource.SystemScriptGraph.NiagaraNodeAssignment_2.SetVariables_#3
 ]
}
emitters {
 Glow {
  emitterUpdateScript [
   &EmitterState(Life Cycle Mode=::NewEnumerator1, Loop Behavior=::NewEnumerator1, Loop Duration Mode=::NewEnumerator1, UseLoopDelay=false)
   &SpawnBurst_Instantaneous(Spawn Probability=HlslExpression() {hlslExpression=1.0}, Use Spawn Probability=false, UseLoopCountLimit=false)
  ]
  particleSpawnScript [
   &PartitionParticles() {enabled=false}
   &InitializeParticle(
    Color=<rgba 0.6719 0.0101 0.3379 0.2>
    Sprite Size Mode=::NewEnumerator3
    Uniform Sprite Size=90
    UsePositionOffset=false
    Write Material Random=true
   )
  ]
  particleUpdateScript [
   &ParticleState(Kill Particles When Lifetime Has Elapsed=false, Let Infinitely Lived Particles Die When Emitter Deactivates=true)
   &SkeletalMeshLocation(
    MaskSpawnGroup=false
    Mesh Sampling Type=::NewEnumerator5
    Sampled Position Offset=<0 0 5>
    Skeletal Mesh=$User.SkeletalMesh
    Socket Sampling Mode=::NewEnumerator5
   )
   &ScaleColor(
    Scale Alpha=&Lerp_Float(
     A=0
     B=&FNTimeOfDayScalar(Night Scale=0.5, Morning Scale=0.5, Evening Scale=0.5)
     Alpha=&FloatFromCurve(FloatCurve=DataInterface() {propertyValues {Curve {keys [
          {interpMode=RCIM_Cubic, tangentMode=RCTM_User, time=0, value=0}
          {arriveTangent=2.4173, interpMode=RCIM_Cubic, leaveTangent=2.4173, tangentMode=RCTM_User, time=1, value=1}
         ]}}})
    )
    ScaleA=true
    ScaleRGB=false
   )
   &CameraOffset(Camera Offset Amount=5)
  ]
  rendererClasses [&NiagaraSpriteRendererProperties]
  renderers [{rendererClass=&NiagaraSpriteRendererProperties}]
 }
 Light {
  bEnabled=false
  emitterSpawnScript [&FN_PreserveSkydiveUserVariables() {enabled=false, module=PreserveUserVariables}]
  emitterUpdateScript [
   &EmitterState()
   &SpawnBurst_Instantaneous(Use Spawn Probability=false, UseLoopCountLimit=false)
  ]
  particleSpawnScript [
   &PartitionParticles() {enabled=false}
   &InitializeParticle(Lifetime=2, Sprite Size Mode=::NewEnumerator3, UsePositionOffset=false, Write Material Random=true)
  ]
  particleUpdateScript [
   &ParticleState(
    Kill Particles When Lifetime Has Elapsed=false
    Let Infinitely Lived Particles Die When Emitter Deactivates=true
    Loop Particles Lifetime=true
   )
   &SkeletalMeshLocation(
    MaskSpawnGroup=false
    Mesh Sampling Type=::NewEnumerator5
    Sampled Position Offset=<0 0 8.7197>
    Skeletal Mesh=$User.SkeletalMesh
    Socket Sampling Mode=::NewEnumerator5
   )
   &FN_Light_Attributes(
    Color=<rgba 0.8632 0.013 0.4342 1>
    Create Color Attribute=true
    Create Exponent Attribute=true
    Create Radius Attribute=true
    Exponent=2
    Radius=100
    Scale Color=&Multiply_Float(
     A=&FNTimeOfDayScalar(Night Scale=2.5, Morning Scale=3, Day Scale=4, Evening Scale=3)
     B=&Waveform(
      Waveform Count=::NewEnumerator2
      "Waveform [1]"=::NewEnumerator0
      "[1] Frequency"=1
      "[1] Amplitude Min/Max"=<0.9 1>
      "[1] Scale"=0.5
      "[2] Average"=false
      "Waveform [2]"=::NewEnumerator5
      "[2] Input Position"=0
      "[2] Phase"=$Emitter.Age
      "[2] Phase Scale"=1
      "[2] Compound Frequency"=2
      "[2] Frequency"=2
      "[2] Amplitude Min/Max"=<0 1>
      "[2] Scale"=0.05
      "[2] Bias"=0
      "[2] Exponent"=1
      "[2] Pulse Duration"=0.3
      "[2] Clip Triangle Min/Max"=<0 0>
      "[2] Random Seed"=0
      "[3] Average"=false
      "Waveform [3]"=::NewEnumerator5
      "[3] Input Position"=0
      "[3] Phase"=$Emitter.Age
      "[3] Phase Scale"=1
      "[3] Frequency"=1
      "[3] Compound Frequency"=2
      "[3] Amplitude Min/Max"=<0 1>
      "[3] Scale"=0.03
      "[3] Bias"=0
      "[3] Exponent"=1
      "[3] Pulse Duration"=0.3
      "[3] Clip Triangle Min/Max"=<0 0>
      "[3] Random Seed"=0
     )
    )
   )
   /SetVariables(Particles.LightIntensity:NiagaraFloat=&Multiply_Float(
     A=20
     B=&Multiply_Float(
      A=&Multiply_Float(
       A=&FNTimeOfDayScalar(Night Scale=0.6, Morning Scale=1, Evening Scale=1)
       B=&FloatFromCurve(FloatCurve=DataInterface() {propertyValues {Curve {keys [
            {arriveTangent=9.0725, interpMode=RCIM_Cubic, time=0, value=0.5}
            {arriveTangent=0.25, interpMode=RCIM_Cubic, leaveTangent=0.25, time=0.1408, value=0.9847}
            {arriveTangent=-0.1661, interpMode=RCIM_Cubic, leaveTangent=-0.1661, time=0.3, value=0.575}
            {arriveTangent=-0.5, interpMode=RCIM_Cubic, leaveTangent=-0.5, time=0.5, value=0.925}
            {arriveTangent=-0.0833, interpMode=RCIM_Cubic, leaveTangent=-0.0833, time=0.6, value=0.425}
            {arriveTangent=0.1875, interpMode=RCIM_Cubic, leaveTangent=0.1875, time=0.8, value=0.9}
            {interpMode=RCIM_Cubic, leaveTangent=-0.0405, time=1, value=0.5}
           ]}}})
      )
      B=&Waveform(
       Waveform Count=::NewEnumerator1
       "Waveform [1]"=::NewEnumerator0
       "[1] Frequency"=1
       "[1] Amplitude Min/Max"=<0.9 1>
       "[1] Scale"=0.5
       "[2] Average"=false
       "[2] Input Position"=0
       "[2] Phase"=$Emitter.Age
       "[2] Phase Scale"=1
       "[2] Compound Frequency"=2
       "[2] Frequency"=2
       "[2] Amplitude Min/Max"=<0 1>
       "[2] Scale"=0.3
       "[2] Bias"=0
       "[2] Exponent"=1
       "[2] Pulse Duration"=0.1
       "[2] Clip Triangle Min/Max"=<0 0>
       "[2] Random Seed"=0
      )
     )
    )) {isSetParameters=true}
  ]
  rendererClasses [
   &NiagaraLightRendererProperties
   &NiagaraSpriteRendererProperties
   &NiagaraComponentRendererProperties
  ]
  renderers [
   {rendererClass=&NiagaraLightRendererProperties, rendererIndex=0}
   {rendererClass=&NiagaraSpriteRendererProperties, rendererIndex=1}
   {rendererClass=&NiagaraComponentRendererProperties, rendererIndex=2}
  ]
 }
 Shoulders_FX {
  bEnabled=false
  emitterUpdateScript [
   &EmitterState(
    Inactive Response=::NewEnumerator2
    Life Cycle Mode=::NewEnumerator1
    Loop Duration=&RandomRangeFloat(Minimum=0.1, Maximum=0.6)
    Recalculate Duration Each Loop=true
    UseLoopDelay=false
   )
   &SpawnBurst_Instantaneous(
    Spawn Count=&RandomRangeInt(Minimum=1, Maximum=3)
    Spawn Probability=$User.ParticlesAmount
    Spawn Time=&RandomRangeFloat(Minimum=0, Maximum=0.1)
    Use Spawn Probability=true
    UseLoopCountLimit=false
   )
   /SetVariables(Emitter.DepthFade=35) {isSetParameters=true, module=SetVariables_#4}
  ]
  particleSpawnScript [
   &FN_SetSpawnGroup(Number of Spawn Groups=1) {enabled=false}
   /SetVariables(
    Particles.MeshIndex=&RandomRangeInt(Minimum=0, Maximum=5)
    Particles.bIsTravelingBolt:NiagaraBool=&SetBoolByFloatComparison(
     A=&RandomRangeFloat(Minimum=0, Maximum=1)
     B=&FN_FrontEndFloatModifier(Gameplay=0.25, Front End=0)
    )
   ) {isSetParameters=true, module=SetVariables_#5}
   /SetVariables(Particles.MaskSize:NiagaraFloat=&MakeCustomFloatFromBool(
     Bool=$Particles.bIsTravelingBolt
     True Float=&RandomRangeFloat(Minimum=0.25, Maximum=0.5)
     False Float=2
    )) {isSetParameters=true}
   /SetVariables(Particles.MeshIndex=5) {enabled=false, isSetParameters=true, module=SetVariables_#6}
   &InitializeParticle(
    Color=&FNTimeOfDayColor(
     Night Color=<rgba 50 0.7521 25.1492 1>
     Morning Color=<rgba 50 0.7521 25.1492 1>
     Day Color=<rgba 50 0.7521 25.1492 1>
     Evening Color=<rgba 50 0.7521 25.1492 1>
    )
    Lifetime=&MakeCustomFloatFromBool(
     Bool=$Particles.bIsTravelingBolt
     True Float=&RandomRangeFloat(Minimum=0.6, Maximum=0.8)
     False Float=&RandomRangeFloat(Minimum=0.33, Maximum=0.22)
    )
    Mesh Scale=&Multiply_VectorByFloat(
     Vector=&SelectVectorFromArray(Direct Array Index=$Particles.SpawnGroup, Vector Selection Array=DataInterface() {propertyValues {FloatData [<1 1 1>, <1 -1 1>]}})
     Float=&RandomRangeFloat(Minimum=0.8, Maximum=1.25)
    )
    Mesh Scale Mode=::NewEnumerator4
    Sprite Rotation Mode=::NewEnumerator2
    Sprite Size Mode=::NewEnumerator3
    Uniform Sprite Size=&Multiply_Float(
     A=&RandomRangeFloat(Minimum=9, Maximum=17)
     B=&SelectFloatFromArray(
      Direct Array Index=$Particles.MeshIndex
      Float Selection Array=DataInterface() {propertyValues {FloatData [1,1,1,1,0,0,1,1,1,1,1,1,1,1,1,1,1,1]}}
     )
    )
    UsePositionOffset=true
    Write Material Random=false
   )
   &DynamicMaterialParameters(
    Index 0 Param 1=&RandomRangeFloat(Minimum=3, Maximum=4)
    Index 0 Param 2=$Particles.MaskSize
    Index 0 Param 4=0.5
    Module.Param0WriteEnabled=true
    Module.Param1WriteEnabled=true
    Module.Param2WriteEnabled=true
    Module.Param3WriteEnabled=true
   )
   &CameraOffset(Camera Offset Amount=2)
   /SetVariables(Particles.ElectricityVisiblityTag:NiagaraInt32=&MakeIntFromFloat(Float=$User.ColorSwitcher)) {enabled=false, isSetParameters=true, module=SetVariables_#7}
   &InitialMeshOrientation(
    Mesh Orientation Mode=::NewEnumerator2
    Rotation=&RandomRangeVector(Minimum=<0 0 -0.1>, Maximum=<0 1 0.1>)
    Use Rotation Vector=true
   ) {enabled=false}
  ]
  particleUpdateScript [
   &SolveForcesAndVelocity(Clamp Velocity=false, Limit Acceleration=false)
   &SkeletalMeshLocation(MaskSpawnGroup=false, Mesh Sampling Type=::NewEnumerator5, Skeletal Mesh=$User.SkeletalMesh, Socket Sampling Mode=::NewEnumerator5)
   /SetVariables(
    Particles.MaskIntensity:NiagaraFloat=&MakeCustomFloatFromBool(
     Bool=$Particles.bIsTravelingBolt
     True Float=2
     False Float=&FloatFromCurve(FloatCurve=DataInterface() {propertyValues {Curve {keys [{tangentMode=RCTM_User, time=0, value=0.333}, {tangentMode=RCTM_User, time=1, value=0.666}]}}})
    )
    Particles.MaskPosition:NiagaraFloat=&MakeCustomFloatFromBool(
     Bool=$Particles.bIsTravelingBolt
     True Float=$Particles.NormalizedAge
     False Float=&FloatFromCurve(FloatCurve=DataInterface() {propertyValues {Curve {keys [<at 0 0>,<at 0.333 0.5>]}}})
    )
    Particles.MeshPosition:NiagaraPosition=$Particles.Position
   ) {isSetParameters=true}
   &ParticleState()
   &DynamicMaterialParameters(
    Index 0 Param 1=&MakeCustomFloatFromBool(
     Bool=$Particles.bIsTravelingBolt
     True Float=$System.LightningWidth
     False Float=&FloatFromCurve(FloatCurve=DataInterface() {propertyValues {Curve {keys [<at 0.666 1>,<at 1 0>]}}}, Scale Curve=$System.LightningWidth)
    )
    Index 0 Param 3=$Particles.MaskPosition
    Index 0 Param 4=$Particles.MaskIntensity
    Module.Param0WriteEnabled=true
    Module.Param1WriteEnabled=false
    Module.Param2WriteEnabled=true
    Module.Param3WriteEnabled=true
   ) {module=DynamicMaterialParameters001}
   &ScaleColor(
    Scale RGB=&VectorFromFloat(Value=&Multiply_Float(A=&Lerp_Float(A=0.133, B=1, Alpha=$System.Flicker), B=1))
    ScaleA=false
    ScaleRGB=true
   )
   &ScaleColor(
    Color Value To Scale=$Particles.Color
    Scale RGB=&Multiply_VectorByFloat(Vector=<1 1 1>, Float=$System.TimeOfDayBrightnessMultiplier)
    ScaleA=true
    ScaleRGB=true
   ) {module=ScaleColor001}
   /SetVariables(
    Particles.CameraOffsetGlow=-0.5
    Particles.ColorGlow:LinearColor=<rgba 0.401 0 0.1722 5>
    Particles.DynamicMaterialParameterGlow:Vector4f=&MakeVector4(
     X=&Add_Float(A=$System.LightningWidth, B=6)
     Y=$Particles.MaskSize
     Z=$Particles.MaskPosition
     W=$Particles.MaskIntensity
    )
   ) {isSetParameters=true}
  ]
  rendererClasses [&NiagaraMeshRendererProperties]
  renderers [
   {rendererClass=&NiagaraMeshRendererProperties, rendererIndex=0}
   {rendererClass=&NiagaraMeshRendererProperties, rendererIndex=1}
  ]
 }
 Shoulders_FX002 {
  emitterUpdateScript [
   &EmitterState(
    Inactive Response=::NewEnumerator2
    Life Cycle Mode=::NewEnumerator1
    Loop Duration=&RandomRangeFloat(Minimum=0.2, Maximum=0.3)
    Recalculate Duration Each Loop=true
    UseLoopDelay=false
   )
   &SpawnBurst_Instantaneous(
    Spawn Count=&RandomRangeInt(Minimum=1, Maximum=1)
    Spawn Probability=$User.ParticlesAmount
    Spawn Time=&RandomRangeFloat(Minimum=0, Maximum=0.1)
    Use Spawn Probability=true
    UseLoopCountLimit=false
   )
   /SetVariables(Emitter.DepthFade=35) {isSetParameters=true, module=SetVariables_#4}
  ]
  particleSpawnScript [
   &PartitionParticles(Number Of Partitions=3) {enabled=false}
   /SetVariables(
    Particles.MeshIndex=&RandomRangeInt(Minimum=0, Maximum=5)
    Particles.bIsTravelingBolt:NiagaraBool=&SetBoolByFloatComparison(
     A=&RandomRangeFloat(Minimum=0, Maximum=1)
     B=&FN_FrontEndFloatModifier(Gameplay=0.25, Front End=0)
    )
   ) {isSetParameters=true, module=SetVariables_#5}
   /SetVariables(
    Particles.MaskSize:NiagaraFloat=&MakeCustomFloatFromBool(
     Bool=$Particles.bIsTravelingBolt
     True Float=&RandomRangeFloat(Minimum=0.25, Maximum=0.5)
     False Float=2
    )
    Particles.SetWidth:NiagaraFloat=&RandomRangeFloat(Minimum=3, Maximum=4)
   ) {isSetParameters=true}
   /SetVariables(Particles.MeshIndex=5) {enabled=false, isSetParameters=true, module=SetVariables_#6}
   &InitializeParticle(
    Color=&FNTimeOfDayColor(
     Night Color=<rgba 50 0.7521 25.1492 1>
     Morning Color=<rgba 50 0.7521 25.1492 1>
     Day Color=<rgba 50 0.7521 25.1492 1>
     Evening Color=<rgba 50 0.7521 25.1492 1>
    )
    Lifetime=&MakeCustomFloatFromBool(
     Bool=$Particles.bIsTravelingBolt
     True Float=&RandomRangeFloat(Minimum=0.6, Maximum=0.8)
     False Float=&RandomRangeFloat(Minimum=0.33, Maximum=0.22)
    )
    Mesh Scale Mode=::NewEnumerator3
    Mesh Uniform Scale=&RandomRangeFloat(Minimum=0.9, Maximum=1.1)
    Sprite Rotation Mode=::NewEnumerator2
    Sprite Size Mode=::NewEnumerator3
    UsePositionOffset=true
    Write Material Random=false
   )
   &ShapeLocation(
    Cylinder Height=0
    Cylinder Radius=4
    MaskSpawnGroup=false
    Offset Coordinate Space=::Local
    Offset Mode=::NewEnumerator0
    Override Seed=false
    Shape Primitive=::NewEnumerator1
    Surface Only=false
    UseOwnerScale=true
   ) {enabled=false}
   &CameraOffset()
   /SetVariables(Particles.ElectricityVisiblityTag:NiagaraInt32=&MakeIntFromFloat(Float=$User.ColorSwitcher)) {enabled=false, isSetParameters=true, module=SetVariables_#7}
   &DynamicMaterialParameters(
    Index 0 Param 2=$Particles.MaskSize
    Index 0 Param 4=3
    Module.Param0WriteEnabled=false
    Module.Param1WriteEnabled=true
    Module.Param2WriteEnabled=false
    Module.Param3WriteEnabled=false
   )
   &InitialMeshOrientation(Mesh Orientation Mode=::NewEnumerator7, Use Rotation Vector=true)
  ]
  particleUpdateScript [
   &ParticleState()
   &SolveForcesAndVelocity(Clamp Velocity=false, Limit Acceleration=false) {enabled=false}
   /SetVariables(
    Particles.MaskIntensity:NiagaraFloat=&MakeCustomFloatFromBool(
     Bool=$Particles.bIsTravelingBolt
     True Float=2
     False Float=&FloatFromCurve(FloatCurve=DataInterface() {propertyValues {Curve {keys [{tangentMode=RCTM_User, time=0, value=0.333}, {tangentMode=RCTM_User, time=1, value=0.666}]}}})
    )
    Particles.MaskPosition:NiagaraFloat=&MakeCustomFloatFromBool(
     Bool=$Particles.bIsTravelingBolt
     True Float=$Particles.NormalizedAge
     False Float=&FloatFromCurve(FloatCurve=DataInterface() {propertyValues {Curve {keys [<at 0 0>,<at 0.333 0.5>]}}})
    )
   ) {isSetParameters=true}
   &SkeletalMeshLocation(
    Bone Sampling Mode=::NewEnumerator4
    MaskSpawnGroup=false
    Sampled Position Offset Coordinate Space:ENiagaraSkelMeshTransforms=::NewEnumerator2
    Skeletal Mesh=$User.SkeletalMesh
   )
   &DynamicMaterialParameters(
    Index 0 Param 1=&MakeCustomFloatFromBool(
     Bool=$Particles.bIsTravelingBolt
     True Float=$Particles.SetWidth
     False Float=&FloatFromCurve(FloatCurve=DataInterface() {propertyValues {Curve {keys [<at 0.666 1>,<at 1 0>]}}}, Scale Curve=$Particles.SetWidth)
    )
    Index 0 Param 3=$Particles.MaskPosition
    Index 0 Param 4=$Particles.MaskIntensity
    Module.Param0WriteEnabled=true
    Module.Param1WriteEnabled=false
    Module.Param2WriteEnabled=true
    Module.Param3WriteEnabled=true
   ) {module=DynamicMaterialParameters001}
   &ScaleColor(
    Scale RGB=&VectorFromFloat(Value=&Multiply_Float(A=&Lerp_Float(A=0.133, B=1, Alpha=$System.Flicker), B=1))
    ScaleA=false
    ScaleRGB=true
   )
   &ScaleColor(
    Color Value To Scale=$Particles.Color
    Scale RGB=&Multiply_VectorByFloat(Vector=<1 1 1>, Float=$System.TimeOfDayBrightnessMultiplier)
    ScaleA=true
    ScaleRGB=true
   ) {enabled=false, module=ScaleColor001}
   /SetVariables(
    Particles.CameraOffsetGlow=-0.5
    Particles.ColorGlow:LinearColor=<rgba 0.401 0 0.1722 5>
    Particles.DynamicMaterialParameterGlow:Vector4f=&MakeVector4(
     X=&Add_Float(
      A=&MakeCustomFloatFromBool(
       Bool=$Particles.bIsTravelingBolt
       True Float=$Particles.SetWidth
       False Float=&FloatFromCurve(FloatCurve=DataInterface() {propertyValues {Curve {keys [<at 0.666 1>,<at 1 0>]}}}, Scale Curve=$Particles.SetWidth)
      )
      B=8
     )
     Y=$Particles.MaskSize
     Z=$Particles.MaskPosition
     W=$Particles.MaskIntensity
    )
   ) {isSetParameters=true}
  ]
  rendererClasses [&NiagaraMeshRendererProperties]
  renderers [
   {rendererClass=&NiagaraMeshRendererProperties, rendererIndex=0}
   {rendererClass=&NiagaraMeshRendererProperties, rendererIndex=1}
  ]
 }
}
summary {
 systemName=NS_Pickaxe_DriftAvatar_Spark_Idle
 emitters [
  {emitterName=Shoulders_FX002, rendererClasses [&NiagaraMeshRendererProperties]}
  {emitterName=Glow, rendererClasses [&NiagaraSpriteRendererProperties]}
  {
   bEnabled=false
   emitterName=Light
   rendererClasses [
    &NiagaraLightRendererProperties
    &NiagaraSpriteRendererProperties
    &NiagaraComponentRendererProperties
   ]
  }
  {bEnabled=false, emitterName=Shoulders_FX, rendererClasses [&NiagaraMeshRendererProperties]}
 ]
 userVariables [
  User.bIsFrontEnd {type=NiagaraBool}
  User.bIsFrontEndPreview {type=NiagaraBool}
  User.ColorSwitcher {type=NiagaraFloat, defaultValue=0, description=""}
  User.FlickerSpeed {type=NiagaraFloat, defaultValue=6.666, description=""}
  User.MaskParticlesAmount {type=NiagaraFloat, defaultValue=0, description=""}
  User.ParticlesAmount {type=NiagaraFloat, defaultValue=1, description=""}
  User.PowersActive {type=NiagaraBool}
  User.SkeletalMesh {
   type=NiagaraDataInterfaceSkeletalMesh
   defaultValue {
    struct=NiagaraExt_VariableValue_DataInterface
    value {dataInterface=&self:NiagaraDataInterfaceSkeletalMesh_0}
   }
  }
 ]
}
